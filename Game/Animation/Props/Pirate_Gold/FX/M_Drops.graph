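# Translucent lit sprite-particle material for water drops, stored with the Pirate Gold prop animation effects.
# A drop normal map offsets the UVs of a reflection texture to fake reflections.
# That reflection is multiplied by particle color, a ColorTint parameter and an Emissive scalar, then passed through the shared emissive adjustment function.
# Drop texture and particle alpha blend an index-of-refraction-style value slightly above one.
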Material /Game/Animation/Props/Pirate_Gold/FX/M_Drops
graph {
 properties {
  bUsedWithParticleSprites=true
  blendMode=BLEND_Translucent
  pixelDepthOffsetMode=PDOM_Legacy
  refractionMethod=RM_IndexOfRefraction
  usedShadingModels=Default Lit
 }
 nodes [
  n1 = ParticleColor()
  n2 = TextureSample() {texture=/Game/VisualThreatManager/StormVisuals/Textures/T_Water_drop_02}
  LinearInterpolate(Alpha=Multiply(A=@n1.A, B=ComponentMask(None=@n2.B) {r=true})) {constA=1, constB=1.1}
  n3 = Multiply(A=VectorParameter() {parameterName=ColorTint, defaultValue=<rgba 0.1887 0.7885 1 1>}.RGB, B=@n1.RGB)
  n4 = ComponentMask(None=TextureSample() {samplerType=SAMPLERTYPE_Normal, texture=/Game/VisualThreatManager/StormVisuals/Textures/T_Water_drop_N_02}.RGB) {g=true, r=true}
  n5 = TextureSample(UVs=Add(A=TextureCoordinate(), B=Multiply(A=@n4) {constB=-1})) {texture=/Game/Effects/Fort_Effects/Gameplay/Water/Textures/FX_Reflection_03_Diff}
  MaterialFunctionCall(Emissive=Multiply(A=Multiply(A=@n3, B=@n5.RGB), B=ScalarParameter() {defaultValue=0.5, parameterName=Emissive})) {materialFunction=&MF_FX_EmissiveAdjustments}
 ]
}
